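; Generic bow arrow projectile, built as a base for child arrows, that varies audio, VFX and impact behavior by charge level.
; Blocking hits play surface-based sounds and leave a stuck-arrow particle on objects or the nearest pawn bone (neck for headshots), skipping vehicles.
; Fully charged hits make that arrow optional, add impact FX and call overridable OnFullyChargedImpact.
; That event can create scheduler objects that time spawns of gameplay actors, projectiles or generic actors.
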(event Custom|AttachPersistentArrowClient
  (params (Component :type Primitive%20Component%20Object%20Reference) (AttachPointName :type Name) (Location :type Vector) (bHitPawn :type Boolean))
  ;[1 1 1] SpawnPersistentArrow: spawns particle effect to make it look like arrow sticks into whatever it hits
  (Networking|CanExecuteCosmeticEvents
    (:True
      (CallFunction|AttachPersistentArrow Component AttachPointName Location bHitPawn))))

(event Custom|SpawnPersistentArrow
  (params (Hit%20Result :type Hit%20Result%20Structure))
  (bind (_bblockinghit _binitialoverlap _time _distance _location _impactpoint _normal _impactnormal _physmat _hitactor _hitcomponent) (Collision|BreakHitResult Hit%20Result))
  ;[1 1 1] SpawnPersistentArrow: spawns particle effect to make it look like arrow sticks into whatever it hits
  (Utilities|Casting|CastToFortAthenaVehicle _hitactor
    (:CastFailed
      ;;[1 1 1] There are some special cases to consider when hitting a pawn
      (bind _asplayer_pawn_athena (Utilities|Casting|CastToPlayerPawn_Athena _hitactor)
        (:then
          (CallFunction|AttachArrowtoPawn _asplayer_pawn_athena))
        (:CastFailed
          ;;[1 1 1] Otherwise, simply attach to the root component of the hit object
          (CallFunction|AttachPersistentArrowClient _hitcomponent "None" _location))))) ; For now, we're not sticking arrows into vehicles because hitting windows will cause floating arrows
)

(event EventBeginPlay
  ;[1 1 1] EVENT BeginPlay
  (|CallParentFunction)
  (Utilities|FlowControl|Sequence
    (:then_0
      ;;[1 1 1] Disable tracer for uncharged shots
      (if (not (CallFunction|CheckCharge))
        (Rendering|SetVisibility (Variables|Default|GetTracerStaticMesh) false false)))
    (:then_1
      ;;[1 1 1] Swap sounds and VFX depending on charge
      (Networking|SwitchHasAuthority
        (:Remote
          (CallFunction|SetSoundsfromCharge)
          (CallFunction|SetVFXFromCharge))))))

(event Custom|ClientPlayImpactSound
  (params (Hit :type Hit%20Result%20Structure))
  (bind (_bblockinghit _binitialoverlap _time _distance _location _impactpoint _normal _impactnormal _physmat) (Collision|BreakHitResult Hit))
  (CallFunction|PlayImpactSound _location _physmat))

(event EventEndPlay
  (params (EndPlayReason :type EEndPlayReason%20Enum :default "Destroyed"))
  ;[1 1 1] EVENT EndPlay
  (|CallParentFunction EndPlayReason)
  (bind _inairaudiocomp (Variables|Default|GetInAirAudioComp)
    (:then
      (Audio|Components|Audio|FadeOut _inairaudiocomp (Variables|Audio|GetInAirSoundFadeOutTime) 0.0 "Linear" :!class /Script/Engine.AudioComponent))
    (:else)))

(event Custom|NetMulticast_PlayFullyChargedImpactFX
  (params (Hit%20Result :type Hit%20Result%20Structure))
  (CallFunction|PlayFullyChargedImpactFX Hit%20Result))

(event Custom|PlayFullyChargedImpactFX
  (params (HitResult :type Hit%20Result%20Structure))
  (bind (_bblockinghit _binitialoverlap _time _distance _location _impactpoint _normal) (Collision|BreakHitResult HitResult))
  ;[1 1 1] Play Impact FX -> Used for impact FX if gameplay cues are not sufficient
  (Effects|Components|ParticleSystem|SpawnAltFXAtLocation "WeaponImpactEffect" (Variables|VFX|GetNS_FullyChargedImpactFX) 0 _location (Math|Conversions|RotationFromXVector _normal)))

(event EventOnTouched
  (params (OtherActor :type Actor%20Object%20Reference) (OtherComp :type Primitive%20Component%20Object%20Reference) (HitResult :type Hit%20Result%20Structure%20%28by%20ref%29) (bIsOverlap :type Boolean))
  (bind (_bblockinghit _binitialoverlap _time _distance _location _impactpoint _normal _impactnormal) (Collision|BreakHitResult (Variables|Default|GetCachedHitResult)))
  (bind _cachedhitresult (Variables|Default|GetCachedHitResult))
  ;[1 1 1] Event On Touched: Validate blocking hit and server authority
  (|CallParentFunction OtherActor OtherComp HitResult bIsOverlap)
  (if (not bIsOverlap)
    (Variables|Default|SetCachedHitResult HitResult)
    ;;[1 1 1] Play impact sounds - this function already has the logic for changing the sounds based on charge, so it can be separated from the subsequent charge check
    (Networking|SwitchHasAuthority
      (:Authority
        (CallFunction|ClientPlayImpactSound HitResult))
      (:Remote
        _))
    (if (CallFunction|CheckCharge)
      (Networking|SwitchHasAuthority
        (:Authority
          ;;[1 1 1] Triggers logic needed for full-charge secondary effects. The impact FX should probably be moved to run directly on clients instead of broadcasting to them but that seems ot be causing issues at the moment - will come back to this soon
          (CallFunction|OnFullyChargedImpact _cachedhitresult)
          (CallFunction|NetMulticastPlayFullyChargedImpactFX _cachedhitresult)
          (if (Variables|VFX|GetbSpawnPersistentArrowOnFullyChargedHit)
            (CallFunction|SpawnPersistentArrow _cachedhitresult)))
        (:Remote
          ;;[1 1 1] VFX for fully charged hits
          (Utilities|FlowControl|Sequence
            (:then_0
              (if (Variables|VFX|GetbUseDifferentCueOnSecondaryEffect)
                (CallFunction|ClientPlaySecondaryEffectImpactCue _location _impactnormal))))))
      (else
        ;;[1 1 1] All uncharged shots should spawn a stuck arrow
        (Networking|SwitchHasAuthority
          (:Authority
            (CallFunction|SpawnPersistentArrow _cachedhitresult)))))))

(event Custom|ClientPlaySecondaryEffectImpactCue
  (params (Location :type Vector) (Normal :type Vector))
  ;[1 1 1] Event to play the gameplay cue associated with the impact of the arrow's secondary effect
  (GameplayCue|ExecuteGameplayCue%28Burst%29Local self (Variables|VFX|GetSecondaryEffectCueTag) (Ability|GameplayCue|MakeGameplayCueParameters :NormalizedMagnitude 0.0 :RawMagnitude 0.0 :Location Location :Normal Normal :Instigator (Actor|GetOwner) :EffectCauser (Actor|GetOwner) :GameplayEffectLevel 1 :AbilityLevel 1 :bReplicateLocationWhenUsingMinimalRepProxy false)))

(event Custom|OnFullyChargedImpact
  (params (Hit%20Result :type Hit%20Result%20Structure))
  ;[1 1 1] OnFullyChargedImpact -> Called after a blocking hit and projectile is fully charged. Override this in children for special hit effects
  ;;[1 1 1] If we should spawn additional actors, cache hit result and start queuing spawns
  (if (Variables|ImpactSpawns|GetbSpawnAdditionalActorsOnImpact)
    (CallFunction|QueueAdditionalSpawns)))

(event Custom|QueueAdditionalSpawns
  ;[1 1 1] Additional spawn logic
  ;;[1 1 1] For each spawn definition, construct a scheduler object. The scheduler object handles timer functionality and sends the NotifySpawn message with pertinent parameters whenever an actor should be spawned. The scheduler object handles its own clean up (unbinding listeners and clearing timers) itself
  (for _array_element (Variables|ImpactSpawns|GetAdditionalSpawns)
    (bind _returnvalue (Game|ConstructObjectfromClass "/PrimalGameplay/Items/Bows/SpawnUtility/B_AdditionalSpawnScheduler.B_AdditionalSpawnScheduler_C" self _array_element))
    (Default|BindEventtoNotifySpawn _returnvalue (delegate Custom|HandleSpawn))
    (Class|BAdditionalSpawnScheduler|StartSchedule _returnvalue)))

(event Custom|HandleSpawn
  (params (SpawnProperties :type FArrow%20Additional%20Spawn%20Properties%20Structure) (bOverrideNumToSpawn :type Boolean) (NumToSpawnOverride :type Integer))
  (bind (_actorclass_2_1dd6278641d02f59adc7b1a1843b155b _actortype_8_24407edb4b046b8f9507d99a2bc49ca9 _numtospawn_21_f81b614a4aed43e706d812ac58824f1c _initialdelay_22_84d4e959441176c5e4d87db9e569192c _bdelaybetweenspawns_12_a60b1f414fc742e5113b81af854de145 _delaybetweenspawns_23_25624f804be8117272aacfa22e533989 _projectilespawnproperties_20_a0fa7ce04ef55b71b588ff9b44691d95) (Utilities|Struct|BreakFArrowAdditionalSpawnProperties SpawnProperties))
  (bind (_bblockinghit _binitialoverlap _time _distance _location) (Collision|BreakHitResult (Variables|Default|GetCachedHitResult)))
  (bind (_bblockinghit_2 _binitialoverlap_2 _time_2 _distance_2 _location_2) (Collision|BreakHitResult (Variables|Default|GetCachedHitResult)))
  ;[1 1 1] Additional spawn logic
  (for _ (range 1 (+ (Math|Integer|SelectInt NumToSpawnOverride (Math|Float|Round (Utilities|Struct|BreakScalableFloat _numtospawn_21_f81b614a4aed43e706d812ac58824f1c)) bOverrideNumToSpawn) 1))
    (switch Utilities|FlowControl|Switch|SwitchonEArrowAdditionalSpawnType _actortype_8_24407edb4b046b8f9507d99a2bc49ca9
      (:NewEnumerator0
        ;;[1 1 1] Spawn additional BGA
        (bind _asbuilding_gameplay_actor (Utilities|Casting|CastToBuildingGameplayActorClass _actorclass_2_1dd6278641d02f59adc7b1a1843b155b))
        (ImpactSpawns|SpawnAdditionalBGA _asbuilding_gameplay_actor _location))
      (:NewEnumerator1
        (bind _asfort_projectile_base (Utilities|Casting|CastToFortProjectileBaseClass _actorclass_2_1dd6278641d02f59adc7b1a1843b155b))
        (ImpactSpawns|SpawnAdditionalProjectile (Variables|Default|GetCachedHitResult) _asfort_projectile_base _projectilespawnproperties_20_a0fa7ce04ef55b71b588ff9b44691d95))
      (:NewEnumerator2
        (CallFunction|SpawnAdditionalGenericActor _actorclass_2_1dd6278641d02f59adc7b1a1843b155b _location_2)))))

(fn AttachArrowToPawn
  (params (Pawn :type Player%20Pawn%20Athena%20Object%20Reference))
  (bind _chosenbone (CallFunction|ChoosePawnAttachBone Pawn))
  (CallFunction|AttachPersistentArrowClient (Variables|/Script/Engine.Character|GetMesh Pawn) _chosenbone "0, 0, 0" true) ; When hitting a pawn, the arrow snaps directly to the bone, so we can leave the location empty
)

(fn AttachPersistentArrow
  (params (Component :type Primitive%20Component%20Object%20Reference) (AttachPointName :type Name) (Location :type Vector) (bHitPawn :type Boolean))
  (bind _returnvalue (Effects|Components|ParticleSystem|SpawnAltFXAttached :Type "GenericAnimNotify" :NiagaraSystemAsset (Variables|VFX|GetNS%20PersistentArrow) :EmitterTemplate (Variables|VFX|GetP_PersistentArrow) :AttachToComponent Component :AttachPointName AttachPointName :Location (select bHitPawn "0, 0, 0" Location) :Scale "1.000000,1.000000,1.000000" :LocationType (select bHitPawn "SnapToTarget" "KeepWorldPosition") :bAutoDestroy true :PoolingMethod "None" :bAutoActivate true))
  (Utilities|IsValid _returnvalue
    (:Is%20Valid
      (Transformation|SetWorldRotation _returnvalue (Math|Conversions|RotationFromXVector (* (Transformation|GetActorForwardVector) -1.0)) false false))))

(fn CheckCharge
  (return (>= (Variables|FortProjectileBase|GetChargePercent) (ScalableFloat|GetValueatLevel (Variables|Default|GetRow_ChargeThreshold)))))

(fn ChoosePawnAttachBone
  (params (Pawn :type Player%20Pawn%20Athena%20Object%20Reference))
  (bind (_bonelocation _returnvalue) (Components|SkinnedMesh|FindClosestBone (Variables|/Script/Engine.Character|GetMesh Pawn) (Transformation|GetActorLocation) 0.0 false))
  ;[1 1 1] Check if we hit the pawn's head - if so, we want to attach to the neck instead. Otherwise, we can just keep the original hit bone
  (return (select (Utilities|Name|Equal(Name) _returnvalue "head") "neck_01" _returnvalue)))

(fn DetermineProjectileDirection
  (params (Origin :type Vector) (BaseDirection :type Vector) (MinAngle :type Float%20%28double-precision%29) (MaxAngle :type Float%20%28double-precision%29))
  (return (Math|Rotator|FindLookatRotation (Variables|GetOrigin) (+ (Variables|GetOrigin) (Math|Vector|RotateVectorAroundAxis (Math|Vector|RotateVectorAroundAxis BaseDirection (Math|Random|RandomFloatinRange MinAngle MaxAngle) "1.000000, 0, 0") (Math|Random|RandomFloatinRange 0.0 356.0) BaseDirection)))))

(fn GetPersistentArrowRotation
  (bind (_returnvalue_roll _returnvalue_pitch _returnvalue_yaw) (Math|Rotator|BreakRotator (Transformation|GetWorldRotation (Variables|Default|GetArrowStaticMesh))))
  (return (Math|Rotator|MakeRotator _returnvalue_roll (+ _returnvalue_pitch 1.8e+02) _returnvalue_yaw)))

(fn GetProperHit
  (params (OriginalHit :type Hit%20Result%20Structure))
  (bind (_bblockinghit _binitialoverlap _time _distance _location _impactpoint _normal) (Collision|BreakHitResult OriginalHit))
  ;[1 1 1] Line trace from 100 units above the original hit to check if the arrow is below the ground
  (bind (_outhit _returnvalue) (Collision|LineTraceForObjects (+ _location (* _normal 5e+01)) _location (Utilities|Array|MakeArray "ObjectTypeQuery1")))
  ;[1 1 1] If we hit a WorldStatic object (the ground) use the new hit, otherwise we can still use the original hit
  (if _returnvalue
    (Development|PrintString "Arrow is underground. Using a hit above" true true "(R=1.000000,G=0.000000,B=0.542902,A=1.000000)")
    (return _outhit)
    (else
      (return (Variables|GetOriginalHit)))))

(fn GetWaterImpactLocationForFX
  (params (WaterSurfaceLocation :type Vector :default "0, 0, 0"))
  (if (CallFunction|CheckCharge)
    _)
  ;[1 1 1] Fully charged shots are moving so fast that the surface location detected by the water interaction component is usually inaccurate - to solve this, we trace from the muzzle of the weapon to the projectile location. If there are no hits, then the original location is valid, otherwise we will find the true point on the surface of the water that the projectile breached
  (bind _asfort_player_pawn (Utilities|Casting|CastToFortPlayerPawn (Variables|Actor|GetInstigator)))
  (bind (_outhit _returnvalue) (Collision|LineTraceForObjects (Effects|GetMuzzleLocation (Variables|/Script/FortniteGame.FortPawn|GetCurrentWeapon _asfort_player_pawn) -1) (+ (* (Math|Vector|GetUnitDirection(Vector) (Effects|GetMuzzleLocation (Variables|/Script/FortniteGame.FortPawn|GetCurrentWeapon _asfort_player_pawn) -1) (Transformation|GetActorLocation)) 1e+02) (Transformation|GetActorLocation)) (Utilities|Array|MakeArray "ObjectTypeQuery13")))
  (if _returnvalue
    (return (.ImpactPoint (Collision|BreakHitResult _outhit)))
    (else
      (return WaterSurfaceLocation))))

(fn PickSoundBasedOnCharge
  (params (AudioSet :type FCharge%20Dependent%20Audio%20Set%20Structure))
  (bind (_defaultsound_2_a355254e47b20d3a480ddba6d5ae4b58 _fullychargedsound_4_546f1d7c406290b8811c5cb3df882822) (Utilities|Struct|BreakFChargeDependentAudioSet AudioSet))
  (return (select (CallFunction|CheckCharge) _fullychargedsound_4_546f1d7c406290b8811c5cb3df882822 _defaultsound_2_a355254e47b20d3a480ddba6d5ae4b58)))

(fn PlayImpactSound
  (params (Location :type Vector) (PhysMat :type Physical%20Material%20Object%20Reference))
  (bind (_value _returnvalue) (Utilities|Map|Find (Variables|Audio|GetImpactSoundSurfaceMap) (Variables|/Script/PhysicsCore.PhysicalMaterial|GetSurfaceType PhysMat)))
  (Development|PrintString "Playing impact sound")
  (Utilities|FlowControl|Sequence
    (:then_0
      (bind _returnvalue_2 (Audio|SpawnSoundatLocation (Variables|Audio|GetImpactSoundDefault) Location))
      (Utilities|IsValid _returnvalue_2
        (:Is%20Valid
          (Utilities|FlowControl|Sequence
            (:then_0
              (Utilities|IsValid PhysMat
                (:Is%20Valid
                  (Audio|Parameter|SetIntegerParameter _returnvalue_2 "SurfaceIndex" (Math|Integer|SelectInt _value 0 _returnvalue) :!class /Script/AudioExtensions.AudioParameterControllerInterface))))
            (:then_1
              (Audio|Parameter|SetIntegerParameter _returnvalue_2 "FullyCharged" (Math|Conversions|ToInteger(Boolean) (CallFunction|CheckCharge)) :!class /Script/AudioExtensions.AudioParameterControllerInterface))))))
    (:then_1
      (bind _inairaudiocomp (Variables|Default|GetInAirAudioComp)
        (:then
          (Audio|Components|Audio|FadeOut _inairaudiocomp (Variables|Audio|GetInAirSoundFadeOutTime) 0.0 "Linear" :!class /Script/Engine.AudioComponent))
        (:else)))
    (:then_2
      (if (CallFunction|CheckCharge)
        (bind _impactsoundlayer (Variables|Audio|GetImpactSoundLayer)
          (:then
            (Audio|PlaySoundatLocation _impactsoundlayer Location))
          (:else))))))

(fn SetSoundsFromCharge
  (bind _inairsounds_defaultsound_2_a355254e47b20d3a480ddba6d5ae4b58 (Utilities|Struct|BreakFChargeDependentAudioSet (Variables|GetInAirSounds_DefaultSound_2_A355254E47B20D3A480DDBA6D5AE4B58)))
  (Utilities|FlowControl|Sequence
    (:then_0
      ;[1 1 1] By design, we're only doing this for non-fully charged arrows now
      (if (not (CallFunction|CheckCharge))
        (bind _returnvalue (Audio|SpawnSoundAttached _inairsounds_defaultsound_2_a355254e47b20d3a480ddba6d5ae4b58 (Variables|Collision|GetCapsuleComponent)))
        (bind _output_get (Variables|Default|SetInAirAudioComp _returnvalue))
        (Audio|Components|Audio|FadeIn _output_get (Variables|Audio|GetInAirSoundFadeInTime) 1.0 0.0 "Linear" :!class /Script/Engine.AudioComponent)))
    (:then_1
      (Variables|Sound|SetBulletWhipTrackerComponentClass (select (and (CallFunction|CheckCharge) (Utilities|IsValidClass (Variables|Audio|GetFullyChargedBulletWhip))) (Variables|Audio|GetFullyChargedBulletWhip) (Variables|Audio|GetNormalBulletWhip))))))

(fn SetVFXFromCharge
  (if (CallFunction|CheckCharge)
    (Niagara|SetNiagaraSystemAsset (Variables|Default|GetNS_ProjectileTrail) (Variables|VFX|GetNS_FullyChargedTrail) true)))

(fn SpawnAdditionalBGA
  (params (BGAClass :type Building%20Gameplay%20Actor%20Class%20Reference) (Location :type Vector))
  (Building|SpawnBuildingGameplayActor BGAClass (Math|Transform|MakeTransform Location) (Game|GetInstigator) false false))

(fn SpawnAdditionalGenericActor
  (params (ActorClass :type Actor%20Class%20Reference) (Location :type Vector))
  (Game|SpawnActorfromClass ActorClass (Math|Transform|MakeTransform Location) "Undefined" "SelectDefaultAtRuntime" 0 (Game|GetInstigator)))

(fn SpawnAdditionalProjectile
  (params (HitResult :type Hit%20Result%20Structure) (ProjectileClass :type Fort%20Projectile%20Base%20Class%20Reference) (ProjectileSpawnProperties :type FProjectile%20Spawn%20Properties%20Structure))
  (bind (_basedirection_13_3b2626524b0172a577c92590347104b1 _busesurfacenormalasbasedirection_16_3cc107b84a714659da12498833199bc6 _minangle_19_cdbb96854d941be4aeda459db3528ebd _maxangle_21_dec5f505488b40770ae7979845b65146) (Utilities|Struct|BreakFProjectileSpawnProperties (Variables|GetProjectileSpawnProperties)))
  (bind (_speed_2_de05d5a3447e141ca9b792835674eea6 _buserangeforspeed_5_023f0e004dea2ecb7d47b89fb72eb325 _minspeed_8_6ad4178d4f25fc52b1ce639be6949490 _maxspeed_10_72d8e5ed486421e0760f76b11b8455de) (Utilities|Struct|BreakFProjectileSpawnProperties (Variables|GetProjectileSpawnProperties)))
  (bind (_binherithiteffectcontainer_24_15e7ec804cc6cbb27f2470bcbf0a08a2 _binheritexplodeeffectcontainer_26_d31a710f4cb6a4e376c467885c8e7eec) (Utilities|Struct|BreakFProjectileSpawnProperties (Variables|GetProjectileSpawnProperties)))
  (bind _returnvalue (Game|GetInstigator))
  (bind (_bblockinghit _binitialoverlap _time _distance _location _impactpoint _normal _impactnormal) (Collision|BreakHitResult HitResult))
  (bind _output_get (LocalVariables|SetSpawnLocation _location))
  (bind _direction (ImpactSpawns|DetermineProjectileDirection _output_get (select _busesurfacenormalasbasedirection_16_3cc107b84a714659da12498833199bc6 _impactnormal _basedirection_13_3b2626524b0172a577c92590347104b1) _minangle_19_cdbb96854d941be4aeda459db3528ebd _maxangle_21_dec5f505488b40770ae7979845b65146))
  (LocalVariables|SetDirection _direction)
  (bind _returnvalue_2 (Projectile|SpawnProjectile (Variables|GetProjectileClass) _returnvalue (LocalVariables|GetSpawnLocation) "0, 0, 0" (LocalVariables|GetDirection) (select _binherithiteffectcontainer_24_15e7ec804cc6cbb27f2470bcbf0a08a2 (Variables|Delivery|GetEffectContainerSpecToApplyOnTouch) 0) (select _binheritexplodeeffectcontainer_26_d31a710f4cb6a4e376c467885c8e7eec (Variables|Delivery|GetEffectContainerSpecToApplyOnExplode) 0) false 0 (select _buserangeforspeed_5_023f0e004dea2ecb7d47b89fb72eb325 (Math|Random|RandomFloatinRange _minspeed_8_6ad4178d4f25fc52b1ce639be6949490 _maxspeed_10_72d8e5ed486421e0760f76b11b8455de) _speed_2_de05d5a3447e141ca9b792835674eea6)))
  (Variables|/Script/Engine.Actor|SetInstigator _returnvalue _returnvalue_2))

(fn ConstructionScript
  (|CallParentFunction))
