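# M_Bow_Proto_Parent is a Default Lit material that, by its name and folder, serves as the parent for prototype bow materials.
# It declares scalar parameters for Metallic, Specular and Roughness plus a grey Base Color vector parameter, with no textures or math nodes.
# Usage flags enable it for skeletal meshes and mesh deformers.
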Material /PrimalGameplay/Items/Bows/M_Bow_Proto_Parent
graph {
 properties {bUsedWithMeshDeformer=true, bUsedWithSkeletalMesh=true, pixelDepthOffsetMode=PDOM_Legacy, usedShadingModels=Default Lit}
 nodes [
  ScalarParameter() {defaultValue=0.5, parameterName=Metallic}
  ScalarParameter() {defaultValue=0.5, parameterName=Specular}
  ScalarParameter() {defaultValue=0.5, parameterName=Roughness}
  VectorParameter() {parameterName=Base Color, defaultValue=<rgba 0.3594 0.3594 0.3594 0>}
 ]
}
